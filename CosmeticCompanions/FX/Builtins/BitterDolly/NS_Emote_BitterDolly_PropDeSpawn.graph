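# Niagara system for the puff effect when the BitterDolly emote's prop despawns, judging by its name and emitters.
# Three toon-style emitters make up the effect: toon_clouds, ToonClouds_LowMedium and ToonImpactSmoke.
# A system spawn script sets variables alongside SystemState, and the system uses time-of-day color and float-array selection dynamic inputs.
# Two user bools flag front-end and front-end preview contexts.
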NiagaraSystem /CosmeticCompanions/FX/Builtins/BitterDolly/NS_Emote_BitterDolly_PropDeSpawn
dependencies {
 usedDataInterfaces [&NiagaraDataInterfaceArrayFloat]
 usedDynamicInputs [&FNTimeOfDayColor, &SelectFloatFromArray]
 usedModules [
  &SystemState
  &self:SystemSpawnScript.SystemScriptSource.SystemScriptGraph.NiagaraNodeAssignment_0.SetVariables_#1
 ]
}
emitters {
 ToonClouds_LowMedium {rendererClasses [], renderers []}
 ToonImpactSmoke {rendererClasses [], renderers []}
 toon_clouds {rendererClasses [], renderers []}
}
summary {
 systemName=NS_Emote_BitterDolly_PropDeSpawn
 emitters [
  {emitterName=toon_clouds, rendererClasses []}
  {emitterName=ToonClouds_LowMedium, rendererClasses []}
  {emitterName=ToonImpactSmoke, rendererClasses []}
 ]
 userVariables [
  User.bIsFrontEnd {type=NiagaraBool}
  User.bIsFrontEndPreview {type=NiagaraBool}
 ]
}
