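# Niagara effect for the prop spawn moment of the BitterDolly emote, part of the built-in cosmetic companion FX.
# Four emitters, named for a spiral, toon clouds, low-medium toon smoke and impact smoke, make up the system; their renderers are not listed.
# A system spawn assignment accompanies SystemState, with time-of-day color and float-array selection dynamic inputs in use.
# Two user booleans flag front-end and front-end preview contexts.
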NiagaraSystem /CosmeticCompanions/FX/Builtins/BitterDolly/NS_Emote_BitterDolly_PropSpawn
dependencies {
 usedDataInterfaces [&NiagaraDataInterfaceArrayFloat]
 usedDynamicInputs [&FNTimeOfDayColor, &SelectFloatFromArray]
 usedModules [
  &SystemState
  &self:SystemSpawnScript.SystemScriptSource.SystemScriptGraph.NiagaraNodeAssignment_0.SetVariables_#1
 ]
}
emitters {
 Spiral002 {rendererClasses [], renderers []}
 ToonImpactSmoke {rendererClasses [], renderers []}
 ToonSmoke_LowMedium {rendererClasses [], renderers []}
 toon_clouds {rendererClasses [], renderers []}
}
summary {
 systemName=NS_Emote_BitterDolly_PropSpawn
 emitters [
  {emitterName=Spiral002, rendererClasses []}
  {emitterName=toon_clouds, rendererClasses []}
  {emitterName=ToonSmoke_LowMedium, rendererClasses []}
  {emitterName=ToonImpactSmoke, rendererClasses []}
 ]
 userVariables [
  User.bIsFrontEnd {type=NiagaraBool}
  User.bIsFrontEndPreview {type=NiagaraBool}
 ]
}
