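# Niagara system for the Bitter Dolly emote's summoning circle: candles, a connecting triangle and warm auras, with a late dark aura and smoke.
# Candle, glow and smoke particles split three ways onto sockets from a shared skeletal mesh reader; triangle meshes orient between those points.
# Spawn-group switches merge several effects per emitter.
# Loop delays stage the sequence, while time-of-day scalars and user animation curves modulate brightness and motion.
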NiagaraSystem /CosmeticCompanions/FX/Builtins/BitterDolly/NS_Emote_BitterDolly_summoningCircle
dependencies {
 usedDataInterfaces [
  &NiagaraDataInterfaceCurve
  &NiagaraDataInterfaceColorCurve
  &NiagaraDataInterfaceSkeletalMesh
  &NiagaraDataInterfaceArrayFloat
 ]
 usedDynamicInputs [&SelectFloatFromArray]
 usedModules [
  &SystemState
  &self:SystemSpawnScript.SystemScriptSource.SystemScriptGraph.NiagaraNodeAssignment_0.SetVariables_#1
  &self:SystemSpawnScript.SystemScriptSource.SystemScriptGraph.NiagaraNodeAssignment_1.SetVariables_#2
  &self:SystemSpawnScript.SystemScriptSource.SystemScriptGraph.NiagaraNodeAssignment_2.SetVariables_#3
  &self:SystemSpawnScript.SystemScriptSource.SystemScriptGraph.NiagaraNodeAssignment_3.SetVariables_#4
 ]
}
emitters {
 Candle_Prop_spawn_merged_Blend_space {
  emitterUpdateScript [
   &EmitterState(Life Cycle Mode=::NewEnumerator1, Loop Behavior=::NewEnumerator2, Loop Count=2, Loop Delay=0.47, Loop Duration=9.2, UseLoopDelay=true)
   /Game/Effects/Niagara/Modules/BlendSpace/BlendSpaceHelper()
   &SpawnBurst_Instantaneous(Spawn Count=30, Spawn Group=2, Use Spawn Probability=false) {module=SpawnBurst_Instantaneous002}
   &SpawnBurst_Instantaneous(Spawn Count=1, Spawn Group=1, Use Spawn Probability=false)
   &SpawnBurst_Instantaneous(Spawn Count=10, Use Spawn Probability=false) {module=SpawnBurst_Instantaneous001}
  ]
  particleSpawnScript [
   /SetVariables(Particles.VisibilityTag=&SpawnGroup()) {isSetParameters=true}
   &PartitionParticles(Number Of Partitions=3)
   /SetVariables(Particles.InitSpriteRotRate=&UniformRangedFloat(Minimum=-1, Maximum=1)) {isSetParameters=true}
   &InitializeParticle(
    Color=&MergeHelper_ColorSwitch(
     0=&FNTimeOfDayColor(
      Night Color=<rgba 0.1353 0.2271 0.6337 1>
      Morning Color=<rgba 0.7309 0.4366 0.3997 1>
      Day Color=<rgba 0.2448 0.2448 0.2448 1>
      Evening Color=<rgba 0.7156 0.5981 0.7911 1>
     )
     1=&FNTimeOfDayColor(
      Night Color=<rgba 0.1353 0.2271 0.6337 1>
      Morning Color=<rgba 0.7309 0.4366 0.3997 1>
      Day Color=<rgba 0.2448 0.2448 0.2448 1>
      Evening Color=<rgba 0.7156 0.5981 0.7911 1>
     )
     2=<rgba 5 3.7043 1.51 1>
     3=?
     Index=&SpawnGroup()
     Num=::NewEnumerator1
    )
    Lifetime=&MergeHelper_FloatSwitch(
     0=&UniformRangedFloat(Minimum=0.2, Maximum=0.45)
     1=0.25
     2=&UniformRangedFloat(Minimum=0.3, Maximum=0.7)
     3=?
     Index=&SpawnGroup()
     Num=::NewEnumerator1
    )
    Mesh Scale Mode=::NewEnumerator3
    Position Mode=::NewEnumerator0
    Sprite Rotation Mode=::NewEnumerator2
    Sprite Size Mode=::NewEnumerator3
    Uniform Sprite Size=&MergeHelper_FloatSwitch(
     0=&UniformRangedFloat(Minimum=0, Maximum=0)
     1=0
     2=&RandomRangeFloat(Minimum=0.5, Maximum=10)
     3=?
     Index=&SpawnGroup()
     Num=::NewEnumerator1
    )
    Write Material Random=true
   )
   &ShapeLocation(
    MaskSpawnGroup=false
    Override Seed=false
    Sphere Radius=&MergeHelper_FloatSwitch(0=7, 1=0, 2=10, 3=?, Index=&SpawnGroup(), Num=::NewEnumerator1)
    Sphere Surface Distribution=1
    UseOwnerScale=true
   )
   /SetVariables(
    Particles.Smoke Velocity:Vector3f=&Multiply_Vector(
     A=$Particles.ShapeLocation.ShapeNormal
     B=&VectorFromFloat(Value=&RandomRangeFloat(Minimum=15, Maximum=50))
    )
    Particles.Sparkles Velocity:Vector3f=&UniformRangedVector(Minimum=<-500 -200 -200>, Maximum=<5000 200 200>)
    Particles.airglow CameraOffset=10
    Particles.smoke CameraOffset:NiagaraFloat=&RandomRangeFloat(Minimum=10, Maximum=25)
   ) {isSetParameters=true}
   &AccelerationForce(Acceleration=&MergeHelper_VectorSwitch(
     0=&UniformRangedVector(Minimum=<0 0 -25>, Maximum=<0 0 -75>)
     1=<0 0 0>
     2=&RandomRangeVector(Minimum=<0 0 0>, Maximum=<0 0 0>)
     3=?
     Index=&SpawnGroup()
     Num=::NewEnumerator1
    ))
   &ApplyInitialForces()
   &AddVelocity(Velocity=&MergeHelper_VectorSwitch(
     0=&Multiply_Vector(
      A=$Particles.ShapeLocation.ShapeNormal
      B=&VectorFromFloat(Value=&RandomRangeFloat(Minimum=15, Maximum=50))
     )
     1=<0 0 0>
     2=&UniformRangedVector(Minimum=<-150 -150 -150>, Maximum=<150 150 150>)
     3=?
     Index=&SpawnGroup()
     Num=::NewEnumerator1
    ))
  ]
  particleUpdateScript [
   &ParticleState()
   &SkeletalMeshLocation(
    Bone / Socket Index=$Particles.PartitionParticles.Partition
    MaskSpawnGroup=false
    Mesh Sampling Type=::NewEnumerator5
    Position Sampling=::NewEnumerator1
    Skeletal Mesh=$System.Skeletal Mesh Reader
    Socket Sampling Mode=::NewEnumerator5
   )
   &ScaleMeshSize(Scale Factor=&VectorFromFloat(Value=&FloatFromCurve(
      FloatCurve=DataInterface() {propertyValues {Curve {keys [
          {arriveTangent=5.3301, interpMode=RCIM_Cubic, leaveTangent=5.3301, tangentMode=RCTM_User, time=0, value=0}
          {interpMode=RCIM_Cubic, tangentMode=RCTM_User, time=0.7, value=0.95}
          {tangentMode=RCTM_User, time=1, value=1}
         ]}}}
      Scale Curve=0.5
     )))
   &SpriteRotationRate(Rotation Rate=&MergeHelper_FloatSwitch(
     0=&Multiply_Float(
      A=$Particles.InitSpriteRotRate
      B=&FloatFromCurve(
       FloatCurve=DataInterface() {propertyValues {Curve {keys [
           {arriveTangent=-1.9214, interpMode=RCIM_Cubic, leaveTangent=-1.9214, tangentMode=RCTM_User, time=-0.0046, value=1.0113}
           {arriveTangent=-0.2496, interpMode=RCIM_Cubic, leaveTangent=-0.2496, tangentMode=RCTM_User, time=0.9896, value=0.16}
          ]}}}
       CurveIndex=$Emitter.Age
      )
     )
     1=0
     2=5
     3=?
     Index=&SpawnGroup()
     Num=::NewEnumerator1
    ))
   &ScaleSpriteSize(
    Scale Sprite Size Mode=::NewEnumerator0
    Uniform Scale Factor=&MergeHelper_FloatSwitch(
     0=&FloatFromCurve(FloatCurve=DataInterface() {propertyValues {Curve {keys [
          {arriveTangent=5.2354, interpMode=RCIM_Cubic, leaveTangent=5.2354, tangentMode=RCTM_User, time=-0.0046, value=0.25}
          <at 0.1439 1.4923>
          <at 1.0011 3.4809 RCIM_Cubic>
         ]}}})
     1=&FloatFromCurve(FloatCurve=DataInterface() {propertyValues {Curve {
         defaultValue=1
         keys [{interpMode=RCIM_Cubic, tangentMode=RCTM_User, time=0, value=0}, {interpMode=RCIM_Cubic, tangentMode=RCTM_User, time=1, value=1}]
        }}})
     2=&FloatFromCurve(
      FloatCurve=DataInterface() {propertyValues {Curve {keys [
          {interpMode=RCIM_Cubic, tangentMode=RCTM_Break, time=-0.0046, value=0}
          {interpMode=RCIM_Cubic, tangentMode=RCTM_Break, time=0.1746, value=0.9952}
          {interpMode=RCIM_Cubic, tangentMode=RCTM_Break, time=1.0011, value=0.002}
         ]}}}
      Scale Curve=0.8
     )
     3=?
     Index=&SpawnGroup()
     Num=::NewEnumerator1
    )
   )
   &ScaleColor(
    Scale Alpha=&MergeHelper_FloatSwitch(
     0=&FloatFromCurve(FloatCurve=DataInterface() {propertyValues {Curve {keys [
          {interpMode=RCIM_Cubic, tangentMode=RCTM_Break, time=0, value=0}
          {interpMode=RCIM_Cubic, tangentMode=RCTM_Break, time=0.2049, value=0.9765}
          {interpMode=RCIM_Cubic, tangentMode=RCTM_Break, time=0.7791, value=0.9107}
          {interpMode=RCIM_Cubic, tangentMode=RCTM_Break, time=1, value=0}
         ]}}})
     1=&FloatFromCurve(FloatCurve=DataInterface() {propertyValues {Curve {keys [
          {interpMode=RCIM_Cubic, tangentMode=RCTM_Break, time=0, value=0}
          {interpMode=RCIM_Cubic, tangentMode=RCTM_Break, time=0.25, value=0.75}
          {interpMode=RCIM_Cubic, tangentMode=RCTM_Break, time=1, value=0}
         ]}}})
     2=&FloatFromCurve(FloatCurve=DataInterface() {propertyValues {Curve {keys [
          {interpMode=RCIM_Cubic, tangentMode=RCTM_User, time=-0.0092, value=0.0144}
          {interpMode=RCIM_Cubic, tangentMode=RCTM_User, time=0.0967, value=0.9767}
          {interpMode=RCIM_Cubic, tangentMode=RCTM_User, time=1.0011, value=0.002}
         ]}}})
     3=?
     Index=&SpawnGroup()
     Num=::NewEnumerator1
    )
    Scale RGB=&MergeHelper_VectorSwitch(0=<1 1 1>, 1=<1 1 1>, 2=<1 1 1>, 3=?, Index=&SpawnGroup(), Num=::NewEnumerator1)
    ScaleA=true
    ScaleRGB=true
   )
   &CurlNoiseForce(
    Bias Noise=false
    Coordinate Space:ENiagaraCoordinateSpace=::Local
    Noise Frequency=&MergeHelper_FloatSwitch(0=0, 1=0, 2=0.02, 3=?, Index=&SpawnGroup(), Num=::NewEnumerator1)
    Noise Quality / Cost=::NewEnumerator0
    Noise Strength=&MergeHelper_FloatSwitch(0=0, 1=0, 2=500, 3=?, Index=&SpawnGroup(), Num=::NewEnumerator1)
    Pan Noise=false
    Random Seed=0
    Randomize Noise Sample=false
    Write to Intrinsic Parameters=false
   )
   &Drag(
    Drag=&MergeHelper_FloatSwitch(
     0=&RandomRangeFloat(Minimum=9, Maximum=13)
     1=0
     2=&RandomRangeFloat(Minimum=1, Maximum=5)
     3=?
     Index=&SpawnGroup()
     Num=::NewEnumerator1
    )
    Use Linear Drag=true
    Use Rotational Drag=true
   )
   &GravityForce(Gravity=&MergeHelper_VectorSwitch(
     0=<0 0 0>
     1=<0 0 0>
     2=&RandomRangeVector(Minimum=<0 0 50>, Maximum=<0 0 150>)
     3=?
     Index=&SpawnGroup()
     Num=::NewEnumerator1
    ))
   &SolveForcesAndVelocity(Clamp Velocity=false, Limit Acceleration=false)
   /SetVariables(
    Particles.Cloud dynparam:Vector4f=&MakeVector4(
     X=&FloatFromCurve(FloatCurve=DataInterface() {propertyValues {Curve {keys [
          {arriveTangent=2.2672, interpMode=RCIM_Cubic, leaveTangent=2.2672, tangentMode=RCTM_User, time=0, value=0}
          {interpMode=RCIM_Cubic, tangentMode=RCTM_User, time=1, value=1}
         ]}}})
     Y=0
     Z=0
     W=0
    )
    Particles.smoke DynParam:Vector4f=&MakeVector4(
     X=&FloatFromCurve(FloatCurve=DataInterface() {propertyValues {Curve {keys [<at 0 0>,<at 1 1>]}}})
     Y=0
     Z=0
     W=0
    )
   ) {isSetParameters=true}
   /SetVariables(
    Particles.Low med cloud dynparam:Vector4f=&MakeVector4(
     X=&FloatFromCurve(FloatCurve=DataInterface() {propertyValues {Curve {
         defaultValue=1
         keys [{interpMode=RCIM_Cubic, tangentMode=RCTM_User, time=0, value=0}, {interpMode=RCIM_Cubic, tangentMode=RCTM_User, time=1, value=1}]
        }}})
     Y=$System.Picker
     Z=0
     W=0
    )
    Particles.low med cloud Color:LinearColor=&MakeLinearColorFromVectorAndFloat(
     "Vector (RGB)"=&MakeVectorFromLinearColorRGB(Color=$System.CloudColor Low)
     "Float (Alpha)"=&FloatFromCurve(FloatCurve=DataInterface() {propertyValues {Curve {
         defaultValue=1
         keys [{interpMode=RCIM_Cubic, tangentMode=RCTM_User, time=0, value=1}, {interpMode=RCIM_Cubic, tangentMode=RCTM_User, time=1, value=0}]
        }}})
    )
    Particles.low med cloud size:Vector2f=&Vector2DFromFloat(Value=&RandomRangeFloat(Minimum=25, Maximum=30))
   ) {isSetParameters=true}
  ]
  rendererClasses [&NiagaraMeshRendererProperties, &NiagaraSpriteRendererProperties]
  renderers [
   {rendererClass=&NiagaraMeshRendererProperties, rendererIndex=0}
   {rendererClass=&NiagaraSpriteRendererProperties, rendererIndex=1}
   {rendererClass=&NiagaraSpriteRendererProperties, rendererIndex=2}
   {rendererClass=&NiagaraSpriteRendererProperties, rendererIndex=3}
  ]
 }
 Evil_Aura {
  emitterUpdateScript [
   &EmitterState(Life Cycle Mode=::NewEnumerator1, Loop Behavior=::NewEnumerator1, Loop Delay=8.65, Loop Duration=0.3, UseLoopDelay=true) {module=EmitterState001}
   &SpawnBurst_Instantaneous(Spawn Count=3, Spawn Group=1, Use Spawn Probability=false, UseLoopCountLimit=false) {module=SpawnBurst_Instantaneous001}
  ]
  particleSpawnScript [
   &InitializeParticle(
    Color=&Multiply_LinearColorByFloat(
     Linear Color=<rgba 0 0 0 1>
     Float=&FNTimeOfDayScalar(Night Scale=0.5, Morning Scale=0.78, Evening Scale=0.78)
    )
    Lifetime=0.35
    Position Offset=<15 -3 35>
    Position Offset Coordinate Space=::Local
    Sprite Rotation Mode=::NewEnumerator2
    Sprite Size Mode=::NewEnumerator3
    Uniform Sprite Size=100
    UsePositionOffset=true
    Write Material Random=true
   )
   &CameraOffset(Camera Offset Amount=&RandomRangeFloat(Minimum=-15, Maximum=-10))
   /SetVariables(
    Particles.Aura Large Color:LinearColor=<rgba 0.0099 0.0058 0.0198 1>
    Particles.Aura Large Size:Vector2f=&Vector2DFromFloat(Value=30)
   ) {isSetParameters=true}
  ]
  particleUpdateScript [
   &ParticleState()
   &SpriteRotationRate(Rotation Rate=&RandomRangeFloat(Minimum=-120, Maximum=120))
   &ScaleSpriteSize(Uniform Curve Sprite Scale=DataInterface() {propertyValues {Curve {keys [
        {arriveTangentWeight=0.0807, interpMode=RCIM_Cubic, leaveTangentWeight=0.3949, tangentMode=RCTM_User, time=0, value=0}
        {arriveTangent=0.2924, interpMode=RCIM_Cubic, leaveTangent=0.2924, tangentMode=RCTM_User, time=0.0865, value=1.4766}
        {arriveTangent=-0.5874, interpMode=RCIM_Cubic, leaveTangent=-0.5874, time=0.1385, value=0.9532}
        {
         arriveTangent=0.3131
         arriveTangentWeight=0.3423
         interpMode=RCIM_Cubic
         leaveTangent=0.3131
         leaveTangentWeight=0.0493
         tangentMode=RCTM_User
         time=0.8979
         value=1
        }
        <at 1 0 RCIM_Cubic>
       ]}}})
   &SolveForcesAndVelocity(Clamp Velocity=false, Limit Acceleration=false)
   /SetVariables(
    Particles.Scale Magic Large Color:LinearColor=&MakeLinearColorFromVectorAndFloat(
     "Vector (RGB)"=&MakeVectorFromLinearColorRGB(Color=$Particles.Initial.Aura Large Color)
     "Float (Alpha)"=&FloatFromCurve(FloatCurve=DataInterface() {propertyValues {Curve {keys [<at 0 1>,<at 1 1>]}}})
    )
    Particles.Scale Magic Large Size:Vector2f=&Vector2DFromFloat(Value=&Multiply_Float(
      A=&MakeFloatFromVector2d(Vector 2D=$Particles.Initial.Aura Large Size)
      B=&FloatFromCurve(
       FloatCurve=DataInterface() {propertyValues {Curve {keys [
           {arriveTangentWeight=0.0807, interpMode=RCIM_Cubic, leaveTangentWeight=0.3949, tangentMode=RCTM_User, time=0, value=0}
           {arriveTangent=1.1137, interpMode=RCIM_Cubic, leaveTangent=1.1137, time=0.3, value=1}
           {arriveTangentWeight=0.3423, interpMode=RCIM_Cubic, leaveTangentWeight=0.0493, tangentMode=RCTM_User, time=0.8979, value=1}
           <at 1 0 RCIM_Cubic>
          ]}}}
       Scale Curve=5
      )
     ))
   ) {isSetParameters=true}
   /SetVariables(
    Particles.Evil Aura Outer color:LinearColor=&MakeLinearColorFromVectorAndFloat(
     "Vector (RGB)"=&MakeVectorFromLinearColorRGB(Color=<rgba 0.0217 0.0097 0.0448 1>)
     "Float (Alpha)"=&FloatFromCurve(
      FloatCurve=DataInterface() {propertyValues {Curve {
         defaultValue=0
         keys [
          {interpMode=RCIM_Cubic, tangentMode=RCTM_Break, time=0, value=0}
          {interpMode=RCIM_Cubic, leaveTangent=-0.1735, tangentMode=RCTM_Break, time=0.15, value=1.75}
          {arriveTangent=-0.177, interpMode=RCIM_Cubic, leaveTangent=-0.177, tangentMode=RCTM_User, time=0.3, value=1}
          {arriveTangent=-0.15, interpMode=RCIM_Cubic, leaveTangent=-0.15, tangentMode=RCTM_User, time=0.9574, value=1}
         ]
        }}}
      Scale Curve=0.3
     )
    )
    Particles.Evil Outer Aura:LinearColor=&MakeLinearColorFromVectorAndFloat(
     "Vector (RGB)"=&MakeVectorFromLinearColorRGB(Color=<rgba 0.0281 0.0281 0.0281 1>)
     "Float (Alpha)"=&FloatFromCurve(
      FloatCurve=DataInterface() {propertyValues {Curve {
         defaultValue=0
         keys [
          {interpMode=RCIM_Cubic, tangentMode=RCTM_Break, time=0, value=0}
          {interpMode=RCIM_Cubic, leaveTangent=-0.1735, tangentMode=RCTM_Break, time=0.15, value=1.75}
          {arriveTangent=-0.177, interpMode=RCIM_Cubic, leaveTangent=-0.177, tangentMode=RCTM_User, time=0.3, value=1}
          {arriveTangent=-0.15, interpMode=RCIM_Cubic, leaveTangent=-0.15, tangentMode=RCTM_User, time=0.9574, value=1}
         ]
        }}}
      Scale Curve=0.3
     )
    )
    Particles.Evil Outer Aura Size:Vector2f=&Multiply_Vector2D(A=$Particles.SpriteSize, B=&Vector2DFromFloat(Value=1))
    Particles.Evil center Aura Size:Vector2f=&Multiply_Vector2D(A=$Particles.SpriteSize, B=&Vector2DFromFloat(Value=0.8))
   ) {isSetParameters=true}
  ]
  rendererClasses [&NiagaraSpriteRendererProperties]
  renderers [
   {rendererClass=&NiagaraSpriteRendererProperties}
   {rendererClass=&NiagaraSpriteRendererProperties, rendererIndex=1}
  ]
 }
 First_aura {
  emitterUpdateScript [
   &EmitterState(Life Cycle Mode=::NewEnumerator1, Loop Behavior=::NewEnumerator1, Loop Delay=0.5, Loop Duration=8.5, UseLoopDelay=true)
   &SpawnBurst_Instantaneous(Use Spawn Probability=false, UseLoopCountLimit=false) {module=SpawnBurst_Instantaneous003}
   &SpawnBurst_Instantaneous(Spawn Group=2, Use Spawn Probability=false, UseLoopCountLimit=false) {enabled=false, module=SpawnBurst_Instantaneous002}
   &SpawnBurst_Instantaneous(Spawn Group=1, Use Spawn Probability=false, UseLoopCountLimit=false) {module=SpawnBurst_Instantaneous001}
   &SpawnBurst_Instantaneous(Use Spawn Probability=false, UseLoopCountLimit=false) {enabled=false}
   &SpawnBurst_Instantaneous(Spawn Time=4.1, Use Spawn Probability=false, UseLoopCountLimit=false) {module=SpawnBurst_Instantaneous004}
  ]
  particleSpawnScript [
   /SetVariables(Particles.VisibilityTag=&SpawnGroup()) {isSetParameters=true}
   &InitializeParticle(
    Lifetime=8.7
    Mesh Scale=<0.74 0.74 2>
    Mesh Scale Mode=::NewEnumerator4
    Position Offset=<5 -3 19>
    Position Offset Coordinate Space=::Local
    Sprite Rotation Angle=95
    Sprite Rotation Mode=::NewEnumerator0
    Sprite Size=&MergeHelper_Vector2Switch(0=<140 125>, 1=<165 150>, 2=?, 3=?, 4=?, Index=&SpawnGroup(), Num=::NewEnumerator0)
    Sprite Size Mode=::NewEnumerator4
    UsePositionOffset=true
    Write Material Random=true
   )
   &SpriteFacingAndAlignment(Sprite Facing=<0 0 1>, WriteAlignment=true, WriteFacing=true)
   &CameraOffset(Camera Offset Amount=-15)
   &InitialMeshOrientation(Mesh Axis To Orient=<0 0 1>, Mesh Orientation Mode=::NewEnumerator2, Orientation Vector=<0 0 1>, Rotation=<0 0 1>, Use Rotation Vector=false)
   /SetVariables(Particles.Aura SpriteAlignment:Vector3f=<0 0 1>, Particles.SpriteRotation=-17.3476) {isSetParameters=true}
  ]
  particleUpdateScript [
   &ParticleState(Kill Particles When Lifetime Has Elapsed=false, Let Infinitely Lived Particles Die When Emitter Deactivates=true)
   /SetVariables(Particles.ReactiveAlpha=1) {isSetParameters=true}
   /SetVariables(
    Particles.Center Aura Color:LinearColor=&MakeLinearColorFromVectorAndFloat(
     "Vector (RGB)"=&MakeVectorFromLinearColorRGB(Color=<rgba 1 0.4369 0.0275 0.2>)
     "Float (Alpha)"=&FloatFromCurve(
      FloatCurve=DataInterface() {propertyValues {Curve {
         defaultValue=1
         keys [
          {interpMode=RCIM_Cubic, tangentMode=RCTM_User, time=0, value=0}
          {interpMode=RCIM_Cubic, tangentMode=RCTM_User, time=0.1912, value=1}
          {tangentMode=RCTM_User, time=0.9939, value=0}
         ]
        }}}
      Scale Curve=&FNTimeOfDayScalar(Night Scale=0.3, Morning Scale=0.3, Day Scale=0.3, Evening Scale=0.3)
     )
    )
    Particles.Center Aura size:Vector2f=&Multiply_Vector2D(A=$Particles.SpriteSize, B=&Vector2DFromFloat(Value=0.8))
    "Particles.Main center aura Color ":LinearColor=&MakeLinearColorFromVectorAndFloat(
     "Vector (RGB)"=&MakeVectorFromLinearColorRGB(Color=<rgba 1 0.5673 0.0572 0.2>)
     "Float (Alpha)"=&FloatFromCurve(
      FloatCurve=DataInterface() {propertyValues {Curve {
         defaultValue=1
         keys [
          {interpMode=RCIM_Cubic, tangentMode=RCTM_User, time=0, value=0}
          {interpMode=RCIM_Cubic, tangentMode=RCTM_User, time=0.1912, value=1}
          {tangentMode=RCTM_User, time=0.9939, value=0}
         ]
        }}}
      Scale Curve=&FNTimeOfDayScalar(Night Scale=0.3, Morning Scale=0.3, Day Scale=0.3, Evening Scale=0.3)
     )
    )
    Particles.Main center aura size:Vector2f=&Multiply_Vector2D(A=$Particles.SpriteSize, B=&Vector2DFromFloat(Value=0.65))
   ) {isSetParameters=true}
   /SetVariables(Particles.outer Aura Color:LinearColor=&MakeLinearColorFromVectorAndFloat(
     "Vector (RGB)"=&MakeVectorFromLinearColorRGB(Color=<rgba 1 0.1084 0.0004 0.2>)
     "Float (Alpha)"=&FloatFromCurve(
      FloatCurve=DataInterface() {propertyValues {Curve {
         defaultValue=1
         keys [
          {interpMode=RCIM_Cubic, tangentMode=RCTM_User, time=0, value=0}
          {interpMode=RCIM_Cubic, tangentMode=RCTM_User, time=0.1912, value=1}
          {tangentMode=RCTM_User, time=0.9939, value=0}
         ]
        }}}
      Scale Curve=&FNTimeOfDayScalar(Night Scale=0.5, Morning Scale=0.6, Day Scale=0.6, Evening Scale=0.6)
     )
    )) {isSetParameters=true}
   /SetVariables(
    Particles.Side Aura Position:NiagaraPosition=&AddVectorToPosition(Position=$Particles.Initial.Position, Vector=<7.3236 0 3>)
    Particles.Side Aura SpriteFacing:Vector3f=<0 1 0>
    Particles.Side Aura SpriteSize:Vector2f=<165 120>
    Particles.Side aura SpriteAlignment:Vector3f=<0.52 0 1>
   ) {isSetParameters=true}
   &DynamicMaterialParameters(
    Index 0 Param 1=9
    Module.Index0Float0Write=true
    Module.Index0Float1Write=true
    Module.Index0Float2Write=true
    Module.Index0Float3Write=true
   )
   /SetVariables(
    Particles.Top Aura Position:NiagaraPosition=&AddVectorToPosition(Position=$Particles.Initial.Position, Vector=<-14.382 0 0>)
    Particles.Top Aura SpriteFacing:Vector3f=<0 0 1>
    Particles.Top Aura SpriteSize:Vector2f=<100 100>
    Particles.top Aura Sprite Alignment:Vector3f=<-0.17 1 0>
   ) {isSetParameters=true}
  ]
  rendererClasses [&NiagaraSpriteRendererProperties]
  renderers [
   {rendererClass=&NiagaraSpriteRendererProperties}
   {rendererClass=&NiagaraSpriteRendererProperties, rendererIndex=1}
   {rendererClass=&NiagaraSpriteRendererProperties, rendererIndex=2}
   {rendererClass=&NiagaraSpriteRendererProperties, rendererIndex=3}
   {rendererClass=&NiagaraSpriteRendererProperties, rendererIndex=4}
   {rendererClass=&NiagaraSpriteRendererProperties, rendererIndex=5}
   {rendererClass=&NiagaraSpriteRendererProperties, rendererIndex=6}
   {rendererClass=&NiagaraSpriteRendererProperties, rendererIndex=7}
   {rendererClass=&NiagaraSpriteRendererProperties, rendererIndex=8}
  ]
 }
 Triangle002 {
  emitterUpdateScript [
   &EmitterState(
    Inactive Response=::NewEnumerator1
    Life Cycle Mode=::NewEnumerator1
    Loop Behavior=::NewEnumerator1
    Loop Delay=0.5
    Loop Duration=9.5
    UseLoopDelay=true
   )
   &SpawnBurst_Instantaneous(Spawn Count=3, Spawn Time=0.03, Use Spawn Probability=false, UseLoopCountLimit=false) {module=SpawnBurst_Instantaneous002}
  ]
  particleSpawnScript [
   /SetVariables(Particles.VisibilityTag=&SpawnGroup()) {isSetParameters=true, module=SetVariables_#5}
   &PartitionParticles(Number Of Partitions=3, Write Partition to Ribbon ID=::NewEnumerator0)
   &InitializeParticle(
    Color Mode=::NewEnumerator0
    Lifetime=10
    Mass Mode=::NewEnumerator1
    Mesh Scale=&MergeHelper_VectorSwitch(0=<7.95 5 1.6>, 1=<7.9559 5 1.6>, 2=<7.9559 5 1.6>, 3=?, Index=$Particles.PartitionParticles.Partition, Num=::NewEnumerator1)
    Mesh Scale Mode=::NewEnumerator4
    Ribbon Width=3.75
    Ribbon Width Mode=::NewEnumerator1
    Sprite Size Mode=::NewEnumerator3
    UsePositionOffset=false
    Write Material Random=false
   )
   &InitialMeshOrientation(
    Mesh Axis To Orient=<0 0 0>
    Mesh Orientation Mode=::NewEnumerator2
    Orientation Vector=<0 0 0>
    Rotation=&MergeHelper_VectorSwitch(
     0=&RotateVectorByQuaternion(Vector To Rotate=<0 0 1>, Quaternion=<0 0 0.4661 1>)
     1=<0 0 0>
     2=<0 0 0>
     3=?
     Index=$Particles.PartitionParticles.Partition
     Num=::NewEnumerator1
    )
    Rotation Coordinate Space=::NewEnumerator2
    Use Rotation Vector=true
   )
  ]
  particleUpdateScript [
   &ParticleState()
   &SkeletalMeshLocation(
    MaskSpawnGroup=false
    Mesh Sampling Type=::NewEnumerator5
    Orientation Sampling=::NewEnumerator1
    Skeletal Mesh=$System.Skeletal Mesh Reader
    Socket Sampling Mode=::NewEnumerator5
   )
   &SkeletalMeshLocation(
    Bone / Socket Index=1
    MaskSpawnGroup=false
    Mesh Sampling Type=::NewEnumerator5
    Orientation Sampling=::NewEnumerator1
    Skeletal Mesh=$System.Skeletal Mesh Reader
    Socket Sampling Mode=::NewEnumerator5
   ) {module=SkeletalMeshLocation001}
   &SkeletalMeshLocation(
    Bone / Socket Index=2
    MaskSpawnGroup=false
    Mesh Sampling Type=::NewEnumerator5
    Orientation Sampling=::NewEnumerator1
    Skeletal Mesh=$System.Skeletal Mesh Reader
    Socket Sampling Mode=::NewEnumerator5
   ) {module=SkeletalMeshLocation002}
   /SetVariables(Particles.Position=&AddVectorToPosition(
     Position=&MergeHelper_VectorSwitch(
      0=Linked() {linkedVariable {name=Particles.SkeletalMeshLocation.SampledPosition, type=NiagaraPosition}}
      1=Linked() {linkedVariable {name=Particles.SkeletalMeshLocation001.SampledPosition, type=NiagaraPosition}}
      2=Linked() {linkedVariable {name=Particles.SkeletalMeshLocation002.SampledPosition, type=NiagaraPosition}}
      3=?
      Index=$Particles.PartitionParticles.Partition
      Num=::NewEnumerator1
     )
     Vector=<0 0 -7.5271>
    )) {isSetParameters=true}
   &UpdateMeshOrientation(
    Facing Direction=&MergeHelper_VectorSwitch(
     0=&Subtract_Position(A=$Particles.SkeletalMeshLocation002.SampledPosition, B=$Particles.SkeletalMeshLocation.SampledPosition)
     1=&Subtract_Position(A=$Particles.SkeletalMeshLocation.SampledPosition, B=$Particles.SkeletalMeshLocation001.SampledPosition)
     2=&Subtract_Position(A=$Particles.SkeletalMeshLocation001.SampledPosition, B=$Particles.SkeletalMeshLocation002.SampledPosition)
     3=?
     Index=$Particles.PartitionParticles.Partition
     Num=::NewEnumerator1
    )
    Facing Interpolation=false
    Orientation Method=::NewEnumerator1
    Override Quat=false
    Reference Axis Vector:Vector3f=<0 0 1>
    Stabilization Mode:ENiagara_OrientVectorCount=::NewEnumerator0
    TransformFacing=false
   ) {module=UpdateMeshOrientation001}
   &ScaleColor(
    Scale Alpha=&Multiply_Float(
     A=&FloatFromCurve(FloatCurve=DataInterface() {propertyValues {Curve {keys [
          <at 0 1>
          {arriveTangent=-0.1873, interpMode=RCIM_Cubic, leaveTangent=-0.1873, tangentMode=RCTM_User, time=0.5, value=1}
          {arriveTangent=-2.1957, interpMode=RCIM_Cubic, leaveTangent=-2.1957, tangentMode=RCTM_User, time=0.95, value=0}
         ]}}})
     B=&FloatFromCurve(
      FloatCurve=DataInterface() {propertyValues {Curve {defaultValue=1, keys [{tangentMode=RCTM_User, time=0, value=1}, {tangentMode=RCTM_User, time=1, value=0}]}}}
      CurveIndex=$Emitter.NormalizedLoopAge
     )
    )
    Scale RGB=&VectorFromFloat(Value=&FNTimeOfDayScalar(Night Scale=1, Morning Scale=0.75, Evening Scale=0.75))
    ScaleA=true
    ScaleRGB=true
   )
   /SetVariables(Particles.Triangle color:LinearColor=&MakeLinearColorFromVectorAndFloat(
     "Vector (RGB)"=&MakeVectorFromLinearColorRGB(Color=&ColorFromCurve(ColorCurve=DataInterface() {propertyValues {
         AlphaCurve {keys [<at 0 0.0008>, {arriveTangent=-0.0008, interpMode=RCIM_Cubic, leaveTangent=-0.0008, time=0.5406, value=1}, <at 1 0 RCIM_Cubic>]}
         BlueCurve {keys [<at 0.5193 0.2182 RCIM_Cubic>]}
         GreenCurve {keys [<at 0.5193 0.6832 RCIM_Cubic>]}
         RedCurve {keys [<at 0.5193 1 RCIM_Cubic>]}
        }}))
     "Float (Alpha)"=&FloatFromCurve(
      FloatCurve=DataInterface() {propertyValues {Curve {
         defaultValue=1
         keys [
          {interpMode=RCIM_Cubic, tangentMode=RCTM_User, time=0, value=0}
          {interpMode=RCIM_Cubic, tangentMode=RCTM_User, time=0.0308, value=1.2243}
          {arriveTangent=-0.9575, interpMode=RCIM_Cubic, leaveTangent=-0.9575, time=0.1046, value=0.4604}
          {arriveTangent=-0.6941, interpMode=RCIM_Cubic, leaveTangent=-0.6941, time=0.1578, value=1.1027}
          {arriveTangent=1.2743, interpMode=RCIM_Cubic, leaveTangent=1.2743, time=0.2432, value=0.3642}
          {arriveTangent=0.3307, interpMode=RCIM_Cubic, leaveTangent=0.3307, time=0.3161, value=1.3044}
          {arriveTangent=-0.2468, interpMode=RCIM_Cubic, leaveTangent=-0.2468, time=0.4781, value=0.4419}
          {arriveTangent=0.0145, interpMode=RCIM_Cubic, leaveTangent=0.0145, time=0.6822, value=1.2141}
          {arriveTangent=-3.8948, leaveTangent=-3.8948, time=0.862, value=0.4474}
          <at 1 0>
         ]
        }}}
      Scale Curve=&Multiply_Float(
       A=&FNTimeOfDayScalar(Night Scale=0.8, Morning Scale=1, Evening Scale=1)
       B=$User.Anim_curve_Triangle_Intensity
      )
     )
    )) {isSetParameters=true}
   /SetVariables(
    Particles.Ground Glow SpriteSize:Vector2f=&Vector2DFromFloat(Value=&Waveform(
      Global Amplitude Scale=35
      "Waveform [1]"=::NewEnumerator0
      "[1] Phase Scale"=$System.Glow pulse speed
      "[1] Frequency"=1
      "[1] Amplitude Min/Max"=<0.9 1>
     ))
    Particles.Ground glow Position:NiagaraPosition=/Niagara/DynamicInputs/Vectors/Position/FindMidPointBetweenPositions(A:NiagaraPosition=<1 0 0>, B:NiagaraPosition=<0 1 0>)
   ) {isSetParameters=true}
  ]
  rendererClasses [&NiagaraMeshRendererProperties, &NiagaraSpriteRendererProperties]
  renderers [
   {rendererClass=&NiagaraMeshRendererProperties, rendererIndex=0}
   {rendererClass=&NiagaraSpriteRendererProperties, rendererIndex=1}
  ]
 }
 glows {
  emitterUpdateScript [
   &EmitterState(
    Inactive Response=::NewEnumerator1
    Life Cycle Mode=::NewEnumerator1
    Loop Behavior=::NewEnumerator1
    Loop Delay=0.5
    Loop Duration=9.5
    UseLoopDelay=true
   )
   &SpawnBurst_Instantaneous(Spawn Group=2, Use Spawn Probability=false, UseLoopCountLimit=false) {module=SpawnBurst_Instantaneous002}
   &SpawnBurst_Instantaneous(Spawn Group=2, Spawn Time=0.03, Use Spawn Probability=false, UseLoopCountLimit=false) {module=SpawnBurst_Instantaneous001}
   &SpawnBurst_Instantaneous(Spawn Count=3, Spawn Group=1, Spawn Time=0.05, Use Spawn Probability=false, UseLoopCountLimit=false)
   &SpawnRate(SpawnRate=180, Use Spawn Probability=false)
  ]
  particleSpawnScript [
   /SetVariables(Particles.VisibilityTag=&SpawnGroup()) {isSetParameters=true, module=SetVariables_#5}
   &PartitionParticles(Number Of Partitions=3, Write Partition to Ribbon ID=::NewEnumerator0)
   &InitializeParticle(
    Color Mode=::NewEnumerator0
    Lifetime=&MergeHelper_FloatSwitch(0=0.25, 1=8.7, 2=8.5, 3=?, Index=&SpawnGroup(), Num=::NewEnumerator1)
    Position Mode=::NewEnumerator1
    Ribbon Width=3
    Ribbon Width Mode=::NewEnumerator1
    Sprite Size Mode=::NewEnumerator3
    Uniform Sprite Size=1
    UsePositionOffset=false
    Write Material Random=true
   )
   &AddVelocity(
    Velocity=&MergeHelper_VectorSwitch(
     0=&MakeVector(X=$User.Anim_curve_velocity_X, Y=0, Z=30)
     1=<0 0 0>
     2=<0 0 0>
     3=?
     Index=&SpawnGroup()
     Num=::NewEnumerator1
    )
    Velocity Speed Scale=&MergeHelper_FloatSwitch(0=1, 1=0, 2=0, 3=0, 4=0, Index=&SpawnGroup(), Num=::NewEnumerator3)
   )
   &CameraOffset(Camera Offset Amount=&MergeHelper_FloatSwitch(0=0, 1=0, 2=0, 3=0, 4=0, Index=&SpawnGroup(), Num=::NewEnumerator3))
  ]
  particleUpdateScript [
   &ParticleState()
   &SkeletalMeshLocation(
    Bone / Socket Index=$Particles.PartitionParticles.Partition
    MaskSpawnGroup=false
    Mesh Sampling Type=::NewEnumerator5
    Position Sampling=::NewEnumerator1
    Sampled Position Offset Coordinate Space:ENiagaraSkelMeshTransforms=::NewEnumerator0
    Skeletal Mesh=$System.Skeletal Mesh Reader
    Socket Sampling Mode=::NewEnumerator5
   )
   &SolveForcesAndVelocity(Clamp Velocity=false, Limit Acceleration=false)
   &ScaleColor(
    Scale Alpha=&Multiply_Float(
     A=&FloatFromCurve(FloatCurve=DataInterface() {propertyValues {Curve {keys [
          <at 0 1>
          {arriveTangent=-0.1873, interpMode=RCIM_Cubic, leaveTangent=-0.1873, tangentMode=RCTM_User, time=0.5, value=1}
          {arriveTangent=-2.1957, interpMode=RCIM_Cubic, leaveTangent=-2.1957, tangentMode=RCTM_User, time=0.95, value=0}
         ]}}})
     B=&FloatFromCurve(
      FloatCurve=DataInterface() {propertyValues {Curve {defaultValue=1, keys [{tangentMode=RCTM_User, time=0, value=1}, {tangentMode=RCTM_User, time=1, value=0}]}}}
      CurveIndex=$Emitter.NormalizedLoopAge
     )
    )
    Scale RGB=&VectorFromFloat(Value=&FNTimeOfDayScalar(Night Scale=1, Morning Scale=0.75, Evening Scale=0.75))
    ScaleA=true
    ScaleRGB=true
   )
   &DynamicMaterialParameters(
    Index 0 Param 3=&FloatFromCurve(FloatCurve=DataInterface() {propertyValues {Curve {keys [<at 0 0 RCIM_Cubic>, {arriveTangent=3.2503, interpMode=RCIM_Cubic, leaveTangent=3.2503, tangentMode=RCTM_User, time=1, value=1}]}}})
    Module.Index0Float0Write=true
    Module.Index0Float1Write=true
    Module.Index0Float2Write=true
    Module.Index0Float3Write=true
   )
   /SetVariables(
    Particles.Glow Position:NiagaraPosition=&AddVectorToPosition(Position=$Particles.SkeletalMeshLocation.SampledPosition, Vector=<0 0 3.5>)
    Particles.glow color:LinearColor=&MakeLinearColorFromVectorAndFloat(
     "Vector (RGB)"=&MakeVectorFromLinearColorRGB(Color=<rgba 5 2.8907 0.1505 1>)
     "Float (Alpha)"=&FloatFromCurve(FloatCurve=DataInterface() {propertyValues {Curve {keys [<at 0 1>,<at 1 0>]}}}, Scale Curve=0.17)
    )
    Particles.glow size:Vector2f=&Vector2DFromFloat(Value=&Waveform(
      Global Amplitude Scale=$System.Glow size
      "Waveform [1]"=::NewEnumerator0
      "[1] Phase Scale"=$System.Glow pulse speed
      "[1] Frequency"=1
      "[1] Amplitude Min/Max"=<0.9 1>
     ))
   ) {isSetParameters=true, module=SetVariables_#6}
   &FN_Light_Attributes(
    Color=<rgba 5 2.8907 0.1505 1>
    Create Color Attribute=true
    Create Exponent Attribute=true
    Create Position Attribute=true
    Create Radius Attribute=true
    Enable Particle Light Enable=true
    Enable Visibility Tag=true
    Exponent:NiagaraFloat=&FN_FrontEndFloatModifier(Gameplay=1, Front End=8)
    Light Enabled=true
    Position=&AddVectorToPosition(Position=$Engine.Emitter.SimulationPosition, Vector=<-8 0 15>)
    Radius=40
    Scale Color=&Multiply_Float(
     A=&FloatFromCurve(
      FloatCurve=DataInterface() {propertyValues {Curve {
         defaultValue=1
         keys [
          {arriveTangent=9.0725, time=0, value=0}
          {arriveTangent=0.4179, leaveTangent=0.4179, time=0.0538, value=1.0155}
          {arriveTangent=0.1391, leaveTangent=0.1391, time=0.9772, value=1.0513}
          {arriveTangent=-6.184, leaveTangent=-6.184, time=1, value=0}
         ]
        }}}
      Scale Curve=0.5
     )
     B=&Multiply_Float(
      A=&Waveform("Waveform [1]"=::NewEnumerator2, "[1] Phase Scale"=1.5, "[1] Frequency"=0.5, "[1] Amplitude Min/Max"=<0.85 1>, "[1] Scale"=2.6368)
      B=&FNTimeOfDayScalar(Night Scale=0.8, Morning Scale=0.8, Evening Scale=0.6)
     )
    )
    VisibilityTag=2
   ) {module=FN_Light_Attributes001}
   /SetVariables(
    Particles.Flame Top View SpriteFacing:Vector3f=<0 0 1>
    Particles.Flame top view SpriteAlignment:Vector3f=<0 1 0>
    Particles.Flame top view color:LinearColor=&MakeLinearColorFromVectorAndFloat(
     "Vector (RGB)"=&MakeVectorFromLinearColorRGB(Color=<rgba 5 0.558 0.1505 1>)
     "Float (Alpha)"=&FloatFromCurve(FloatCurve=DataInterface() {propertyValues {Curve {keys [<at 0 1>,<at 0.947 0.9756>,<at 1 0>]}}})
    )
    Particles.Flame top view size:Vector2f=&Vector2DFromFloat(Value=&Waveform(
      Global Amplitude Scale=$System.Glow size
      "Waveform [1]"=::NewEnumerator0
      "[1] Phase Scale"=$System.Glow pulse speed
      "[1] Frequency"=2.5
      "[1] Amplitude Min/Max"=<0.5 0.8>
      "[1] Scale"=0.4
     ))
    Particles.flame top view 2 size:Vector2f=&Vector2DFromFloat(Value=&Waveform(
      Global Amplitude Scale=$System.Glow size
      "Waveform [1]"=::NewEnumerator0
      "[1] Phase Scale"=$System.Glow pulse speed
      "[1] Frequency"=2.5
      "[1] Amplitude Min/Max"=<0.5 0.8>
      "[1] Scale"=0.3
     ))
    Particles.flame top view color 2:LinearColor=&MakeLinearColorFromVectorAndFloat(
     "Vector (RGB)"=&MakeVectorFromLinearColorRGB(Color=<rgba 5 2.8907 0.1505 1>)
     "Float (Alpha)"=&FloatFromCurve(FloatCurve=DataInterface() {propertyValues {Curve {keys [<at 0 1>,<at 0.947 0.9756>,<at 1 0>]}}})
    )
   ) {isSetParameters=true}
  ]
  rendererClasses [
   &NiagaraRibbonRendererProperties
   &NiagaraSpriteRendererProperties
   &NiagaraLightRendererProperties
  ]
  renderers [
   {rendererClass=&NiagaraRibbonRendererProperties}
   {rendererClass=&NiagaraSpriteRendererProperties, rendererIndex=1}
   {rendererClass=&NiagaraLightRendererProperties, rendererIndex=2}
   {rendererClass=&NiagaraSpriteRendererProperties, rendererIndex=3}
   {rendererClass=&NiagaraSpriteRendererProperties, rendererIndex=4}
  ]
 }
 smoke001 {
  emitterUpdateScript [
   &EmitterState(
    Inactive Response=::NewEnumerator1
    Life Cycle Mode=::NewEnumerator1
    Loop Behavior=::NewEnumerator1
    Loop Delay=9
    Loop Duration=1.5
    UseLoopDelay=true
   )
   &SpawnRate(SpawnGroup=2, SpawnRate=180, Use Spawn Probability=false) {enabled=false, module=SpawnRate001}
   &SpawnRate(SpawnRate=180, Use Spawn Probability=false)
  ]
  particleSpawnScript [
   /SetVariables(Particles.VisibilityTag=&SpawnGroup()) {isSetParameters=true, module=SetVariables_#5}
   &PartitionParticles(Number Of Partitions=3, Write Partition to Ribbon ID=::NewEnumerator0)
   &InitializeParticle(
    Color Mode=::NewEnumerator0
    Lifetime=&MergeHelper_FloatSwitch(0=0.475, 1=8, 2=0, 3=0, 4=0, Index=&SpawnGroup(), Num=::NewEnumerator3)
    Ribbon Width Mode=::NewEnumerator1
    Sprite Size Mode=::NewEnumerator3
    Uniform Sprite Size=1
    UsePositionOffset=false
    Write Material Random=true
   )
   &AddVelocity(
    Velocity=&MergeHelper_VectorSwitch(0=<-35 0 20>, 1=<0 0 0>, 2=<0 0 0>, 3=?, Index=&SpawnGroup(), Num=::NewEnumerator1)
    Velocity Speed Scale=&MergeHelper_FloatSwitch(0=1, 1=0, 2=0, 3=0, 4=0, Index=&SpawnGroup(), Num=::NewEnumerator3)
   )
  ]
  particleUpdateScript [
   &ParticleState()
   &SkeletalMeshLocation(
    Bone / Socket Index=$Particles.PartitionParticles.Partition
    MaskSpawnGroup=false
    Mesh Sampling Type=::NewEnumerator5
    Position Sampling=::NewEnumerator1
    Sampled Position Offset Coordinate Space:ENiagaraSkelMeshTransforms=::NewEnumerator0
    Skeletal Mesh=$System.Skeletal Mesh Reader
    Socket Sampling Mode=::NewEnumerator5
   )
   &CurlNoiseForce(Bias Noise=false, Pan Noise=false)
   &SolveForcesAndVelocity(Clamp Velocity=false, Limit Acceleration=false)
   &ScaleColor(
    Scale Alpha=&Multiply_Float(
     A=&FloatFromCurve(
      FloatCurve=DataInterface() {propertyValues {Curve {keys [
          <at 0 1>
          {arriveTangent=-0.1873, interpMode=RCIM_Cubic, leaveTangent=-0.1873, tangentMode=RCTM_User, time=0.5, value=1}
          {arriveTangent=-5.2851, interpMode=RCIM_Cubic, leaveTangent=-5.2851, tangentMode=RCTM_User, time=1, value=0}
         ]}}}
      Scale Curve=2
     )
     B=&FloatFromCurve(
      FloatCurve=DataInterface() {propertyValues {Curve {keys [<at 0 1>, {arriveTangent=-5.2851, interpMode=RCIM_Cubic, leaveTangent=-5.2851, tangentMode=RCTM_User, time=1, value=0}]}}}
      CurveIndex=$Emitter.NormalizedLoopAge
      Scale Curve=3
     )
    )
    Scale RGB=&VectorFromFloat(Value=&FNTimeOfDayScalar(Night Scale=0.5, Morning Scale=0.75, Day Scale=3, Evening Scale=0.75))
    ScaleA=true
    ScaleRGB=true
   )
   /SetVariables(
    Particles.glow color:LinearColor=&MakeLinearColorFromVectorAndFloat(
     "Vector (RGB)"=&MakeVectorFromLinearColorRGB(Color=<rgba 5 2.8907 0.1505 1>)
     "Float (Alpha)"=&FloatFromCurve(FloatCurve=DataInterface() {propertyValues {Curve {keys [<at 0 1>,<at 1 0>]}}}, Scale Curve=0.1)
    )
    Particles.glow size:Vector2f=&Vector2DFromFloat(Value=25)
   ) {isSetParameters=true, module=SetVariables_#6}
   &DynamicMaterialParameters(
    Index 0 Param 1=&FloatFromCurve(
     FloatCurve=DataInterface() {propertyValues {Curve {keys [{arriveTangent=-1.8184, interpMode=RCIM_Cubic, leaveTangent=-1.8184, tangentMode=RCTM_User, time=0, value=1}, <at 1 0 RCIM_Cubic>]}}}
     CurveIndex=$Emitter.NormalizedLoopAge
    )
    Index 0 Param 3=1
    Index 0 Param 4=0.125
    Module.Index0Float0Write=true
    Module.Index0Float1Write=true
    Module.Index0Float2Write=true
    Module.Index0Float3Write=true
   ) {module=DynamicMaterialParameters001}
   /SetVariables() {enabled=false, isSetParameters=true}
   &DynamicMaterialParameters(
    Index 0 Param 1=1
    Index 0 Param 2=5
    Index 0 Param 4=1
    Module.Index0Param0Write=true
    Module.Index0Param1Write=true
    Module.Index0Param2Write=true
    Module.Index0Param3Write=true
   ) {enabled=false}
  ]
  rendererClasses [&NiagaraRibbonRendererProperties]
  renderers [
   {rendererClass=&NiagaraRibbonRendererProperties}
   {rendererClass=&NiagaraRibbonRendererProperties, rendererIndex=1}
  ]
 }
}
summary {
 systemName=NS_Emote_BitterDolly_summoningCircle
 emitters [
  {
   emitterName=Candle_Prop_spawn_merged_Blend_space
   rendererClasses [&NiagaraMeshRendererProperties, &NiagaraSpriteRendererProperties]
  }
  {
   emitterName=glows
   rendererClasses [
    &NiagaraRibbonRendererProperties
    &NiagaraSpriteRendererProperties
    &NiagaraLightRendererProperties
   ]
  }
  {emitterName=First_aura, rendererClasses [&NiagaraSpriteRendererProperties]}
  {emitterName=smoke001, rendererClasses [&NiagaraRibbonRendererProperties]}
  {emitterName=Evil_Aura, rendererClasses [&NiagaraSpriteRendererProperties]}
  {emitterName=Triangle002, rendererClasses [&NiagaraMeshRendererProperties, &NiagaraSpriteRendererProperties]}
 ]
 userVariables [
  User.Anim_Curve_ribbon_width {type=NiagaraFloat, defaultValue=1.5, description=""}
  User.Anim_curve_Triangle_Intensity {type=NiagaraFloat, defaultValue=1, description=""}
  User.Anim_curve_velocity_X {type=NiagaraFloat, defaultValue=0, description=""}
  User.bIsFrontEnd {type=NiagaraBool}
  User.bIsFrontEndPreview {type=NiagaraBool}
  User.OwnerLocalVelocity {type=Vector3f, description="", defaultValue {struct=Vector3f, value=<0 0 0>}}
  User.Smoke_Alpha {type=NiagaraFloat, defaultValue=2, description=""}
  User.Smoke_Position_Offset {type=NiagaraFloat, defaultValue=0, description=""}
 ]
}
